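; Debug NPC spawner actor that requests AI spawns when a player pawn overlaps it, with gating for spawn-once, once-per-player and active spawns.
; After playlist data is ready, it spawns through the AI spawner or Living World string-based debug spawning.
; Tracked request IDs select pawns for inventory, skill, patrol, tag and skydive setup, plus optional auto-hire.
; Construction labels it from the cosmetic character; a replicated X shows disabled state.
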(event EventBeginPlay
  ;[1 1 1] If actor spawn location is not specified, set it to my location
  (Utilities|IsValid (Variables|Default|GetActorLocation)
    (:Is%20Not%20Valid
      (Variables|Default|SetActorLocation self))))

(event Collision|EventActorBeginOverlap
  (params (OtherActor :type Actor%20Object%20Reference))
  (bind _returnvalue (Class|FortAthenaLivingWorldManager|GetLivingWorldManager))
  (bind _patrolpointprovider (Variables|Default|PointProvider|GetPatrolPointProvider))
  (bind _staticpointprovider (Variables|Default|PointProvider|GetStaticPointProvider))
  (bind _extraactortospawn (Variables|Default|GetExtraActorToSpawn))
  (Networking|SwitchHasAuthority
    (:Authority
      (CallFunction|UpdateSpawnDisplay)
      (bind _asfort_player_pawn_athena (Utilities|Casting|CastToFortPlayerPawnAthena OtherActor))
      (Variables|InternalData|SetFort%20Player%20Pawn%20Athena _asfort_player_pawn_athena)
      (if (Pawn|IsBotControlled _asfort_player_pawn_athena)
        _
        (elif (Class|FortPlayerPawn|IsTethered _asfort_player_pawn_athena)
          _
          (elif (and (Variables|Default|GetbSpawnOnce) (Variables|InternalData|GetbHasSpawned))
            (Development|PrintString (Utilities|String|Append (Variables|InternalData|GetDisplayName) " (" (Variables|InternalData|GetCode%20Name) ")" " is already spawned.") true true "(R=1.000000,G=0.000000,B=0.000000,A=1.000000)" 1e+01)
            (elif (Variables|Default|GetSpawnOncePerPlayer)
              (bind _allowspawn (CallFunction|ValidatePlayerAlreadySpawnedAIOrNot))
              (if _allowspawn
                (Class|FortAsyncActionCurrentPlaylistReady|PlaylistDataReadyAsync
                  (:OnReady
                    (for _ (range 1 (+ (Variables|Default|GetSpawnsPerOverlap) 1))
                      (CallFunction|SpawnandConfigureAI)
                      (Variables|InternalData|SetbHasSpawned true) ; Used if bSpawnOnce is true
                    )
                    (Utilities|IsValid _patrolpointprovider
                      (:Is%20Valid
                        (Class|FortAthenaLivingWorldPointProviderInterface|EnablePointProvider_DEPRECATED _patrolpointprovider))
                      (:Is%20Not%20Valid
                        _))
                    (Utilities|IsValid _staticpointprovider
                      (:Is%20Valid
                        (Class|FortAthenaLivingWorldPointProviderInterface|EnablePointProvider_DEPRECATED _staticpointprovider))
                      (:Is%20Not%20Valid
                        _))
                    (Utilities|IsValid _returnvalue
                      (:Is%20Valid
                        (Class|FortAthenaLivingWorldManager|RequestEventGeneration _returnvalue))
                      (:Is%20Not%20Valid
                        _))
                    (if (Utilities|IsValidClass _extraactortospawn)
                      (Game|SpawnActorfromClass _extraactortospawn (Transformation|GetActorTransform (Variables|Default|GetExtraActorToSpawnLocation)) "AlwaysSpawn" "MultiplyWithRoot" 0 0))))
                (else
                  (Development|PrintString (Utilities|String|Append (Variables|InternalData|GetDisplayName) " (" (Variables|InternalData|GetCode%20Name) ")" " already has spawned an AI for: " (Utilities|GetDisplayName (Variables|InternalData|GetFort%20Player%20Pawn%20Athena))) true true "(R=1.000000,G=0.000000,B=0.000000,A=1.000000)" 1e+01)))
              (elif (and (Variables|Default|GetbDisableWhileSpawnActive) (CallFunction|HasActiveSpawned))
                (Development|PrintString (Utilities|String|Append (Variables|InternalData|GetDisplayName) " (" (Variables|InternalData|GetCode%20Name) ")" " has an active spawn.") true true "(R=1.000000,G=0.000000,B=0.000000,A=1.000000)" 1e+01)
                (else
                  _)))))))))

(event Custom|SpawnAI
  (bind _returnvalue (Class|FortAthenaLivingWorldManager|GetLivingWorldManager))
  (bind _patrolpointprovider (Variables|Default|PointProvider|GetPatrolPointProvider))
  (bind _staticpointprovider (Variables|Default|PointProvider|GetStaticPointProvider))
  (bind _extraactortospawn (Variables|Default|GetExtraActorToSpawn))
  (Class|FortAsyncActionCurrentPlaylistReady|PlaylistDataReadyAsync
    (:OnReady
      (for _ (range 1 (+ (Variables|Default|GetSpawnsPerOverlap) 1))
        (CallFunction|SpawnandConfigureAI)
        (Variables|InternalData|SetbHasSpawned true) ; Used if bSpawnOnce is true
      )
      (Utilities|IsValid _patrolpointprovider
        (:Is%20Valid
          (Class|FortAthenaLivingWorldPointProviderInterface|EnablePointProvider_DEPRECATED _patrolpointprovider))
        (:Is%20Not%20Valid
          _))
      (Utilities|IsValid _staticpointprovider
        (:Is%20Valid
          (Class|FortAthenaLivingWorldPointProviderInterface|EnablePointProvider_DEPRECATED _staticpointprovider))
        (:Is%20Not%20Valid
          _))
      (Utilities|IsValid _returnvalue
        (:Is%20Valid
          (Class|FortAthenaLivingWorldManager|RequestEventGeneration _returnvalue))
        (:Is%20Not%20Valid
          _))
      (if (Utilities|IsValidClass _extraactortospawn)
        (Game|SpawnActorfromClass _extraactortospawn (Transformation|GetActorTransform (Variables|Default|GetExtraActorToSpawnLocation)) "AlwaysSpawn" "MultiplyWithRoot" 0 0)))))

(event Custom|UpdateSpawnDisplay
  ;[1 1 1] If this is stopped from spawning show an "X"
  (bind _allowspawn (CallFunction|ValidatePlayerAlreadySpawnedAIOrNot))
  (Variables|InternalData|SetbDisabledIsVisible (Math|Boolean|ORBoolean (and (not _allowspawn) (Variables|Default|GetSpawnOncePerPlayer)) (Variables|Default|GetbSpawnOnce) (and (Variables|Default|GetbDisableWhileSpawnActive) (CallFunction|HasActiveSpawned)))))

(event Custom|DebugSpawner_ResetPlayerSpawnCheck
  ;[1 1 1] Debug Cheat to allow to reset the count if needed
  (Utilities|Array|Clear (Variables|InternalData|GetPlayersThatHaveSpawnedAI))
  (CallFunction|UpdateSpawnDisplay))

(fn ConfigureSpawnedPawn
  (params (NewPawn :type Pawn%20Object%20Reference))
  (bind _newpawn (Variables|GetNewPawn))
  (bind _returnvalue (Pawn|GetController _newpawn))
  (bind _returnvalue_2 (Actor|GetComponentbyClass _returnvalue "/Script/FortniteAI.FortAthenaNpcPatrollingComponent"))
  (bind _spawnskilllevel (Variables|Default|GetSpawnSkillLevel))
  (Utilities|FlowControl|Sequence
    (:then_0
      (bind _asbp_phoebe_player_controller (Utilities|Casting|CastToBP_PhoebePlayerController _returnvalue))
      (for _array_element (Variables|Default|GetSpawnInventory)
        (Inventory|GiveItemtoInventoryOwner :InventoryOwner _asbp_phoebe_player_controller :ItemDefinition (Utilities|Struct|BreakPhoebeItemSpawnInfo _array_element) :NumberToGive (.Amount_5_4D1EC5124092BE03C349959AF7CD2DC2 (Utilities|Struct|BreakPhoebeItemSpawnInfo _array_element)) :bNotifyPlayer true :ItemLevel -1 :PickupInstigatorHandle -1 :bUseItemPickupAnalyticEvent false :WeaponAmmoOverride -1)))
    (:then_1
      (bind _asfort_athena_aibot_controller (Utilities|Casting|CastToFortAthenaAIBotController _returnvalue))
      (if (not (Math|Float|NearlyEqual(Float) _spawnskilllevel -1.0))
        (Class|FortAthenaAIBotController|OverrideSkill _asfort_athena_aibot_controller _spawnskilllevel)))
    (:then_2
      (if (and (Utilities|IsValid _returnvalue_2 :!class /Script/Engine.KismetSystemLibrary) (Utilities|IsValid (Variables|Default|GetPatrolPath) :!class /Script/Engine.KismetSystemLibrary))
        (Class|FortAthenaNpcPatrollingComponent|SetPatrolPath _returnvalue_2 (Variables|Default|GetPatrolPath) false)))
    (:then_3
      (Ability|GameplayEffect|AddLooseGameplayTags (Variables|GetNewPawn) (Variables|Default|GetTagsToApplyOnSpawn) true))
    (:then_4
      (if (Variables|Default|GetbShouldStartSkydiving)
        (bind _asfort_player_pawn_athena (Utilities|Casting|CastToFortPlayerPawnAthena (Variables|GetNewPawn)))
        (Parachute|BeginSkydiving _asfort_player_pawn_athena false)))))

(fn GenerateDisplayTextFromCharacterDefinition
  (if (and (Variables|AutoSetText|GetbGenerateDisplayTextFromCharacterDef) (Editor|InEditororPIE))
    (bind _returnvalue (Class|FortAthenaAISpawnerData|CreateComponentListfromClass (Variables|Default|GetAIBotSpawnerData%20Class) self))
    (bind _output_get (Variables|AutoSetText|SetAIBotSpawnerDataList _returnvalue))
    (bind _returnvalue_2 (Class|FortAthenaAISpawnerDataComponentList|GetList _output_get))
    ;[1 1 1] Was surprised at how fast this broke when someone added a new index, so switched it to this.
    (bind _array_element (Utilities|Array|ForEachLoopwithBreak _returnvalue_2)
      (:LoopBody
        (bind _asfort_athena_aispawner_data_component_cosmetic_loadout (Utilities|Casting|CastToFortAthenaAISpawnerDataComponent_CosmeticLoadout _array_element)))
      (:Completed
        _))
    ;[1 1 1] Got tired of people giving code names vs player facing names, this just does some string manipulation under tthe assumption that the CID provided will be in this form CID_A_292_Athena_Commando_F_ExoSuit
    ;
    ; So we grab the last bit and put it on a new line below the DisplayName from the CID
    (Rendering|Components|TextRender|SetText (Variables|Default|GetTextRender) (Utilities|Text|ToText(String) (Utilities|String|Append (Utilities|String|ToString(Text) (ItemDefinition|GetDisplayName (Class|FortAthenaAISpawnerDataComponentAIBotCosmeticBase|GetCosmeticLoadoutCharacter _asfort_athena_aispawner_data_component_cosmetic_loadout))) "
(" (Utilities|String|RightChop (Item|GetPersistentName (Class|FortAthenaAISpawnerDataComponentAIBotCosmeticBase|GetCosmeticLoadoutCharacter _asfort_athena_aispawner_data_component_cosmetic_loadout)) (+ (Utilities|String|FindSubstring (Item|GetPersistentName (Class|FortAthenaAISpawnerDataComponentAIBotCosmeticBase|GetCosmeticLoadoutCharacter _asfort_athena_aispawner_data_component_cosmetic_loadout)) "_" false true) 1)) ")")))
    (Variables|InternalData|SetDisplayName (Utilities|String|ToString(Text) (ItemDefinition|GetDisplayName (Class|FortAthenaAISpawnerDataComponentAIBotCosmeticBase|GetCosmeticLoadoutCharacter _asfort_athena_aispawner_data_component_cosmetic_loadout))))
    (Variables|InternalData|SetCode%20Name (Utilities|String|RightChop (Item|GetPersistentName (Class|FortAthenaAISpawnerDataComponentAIBotCosmeticBase|GetCosmeticLoadoutCharacter _asfort_athena_aispawner_data_component_cosmetic_loadout)) (+ (Utilities|String|FindSubstring (Item|GetPersistentName (Class|FortAthenaAISpawnerDataComponentAIBotCosmeticBase|GetCosmeticLoadoutCharacter _asfort_athena_aispawner_data_component_cosmetic_loadout)) "_" false true) 1)))))

(fn HasActiveSpawned
  (return (> (Utilities|Array|Length (Variables|InternalData|GetSpawnedPawns)) 0)))

(fn ModifySpawnerComponent
  (params (ComponentList :type Fort%20Athena%20AISpawner%20Data%20Component%20List%20Object%20Reference))
  (return ComponentList))

(fn OnPawnDestroyed
  (params (DestroyedActor :type Actor%20Object%20Reference))
  (bind _aspawn (Utilities|Casting|CastToPawn DestroyedActor))
  (Utilities|Array|RemoveItem (Variables|InternalData|GetSpawnedPawns) _aspawn)
  (Development|PrintString (Utilities|String|Append "@@@ On Pawn Destroyed [BP_DebugNPCSpawner]: CALLED, Spawn count is: " (Utilities|String|ToString(Integer) (Utilities|Array|Length (Variables|InternalData|GetSpawnedPawns))) "!") false (Variables|Debug|GetbDebugSpawner))
  (CallFunction|UpdateSpawnDisplay))

(fn OnPawnSpawned
  (params (NewPawn :type Pawn%20Object%20Reference) (RequestId :type Integer))
  (bind _newpawn (Variables|GetNewPawn))
  (Utilities|Array|AddUnique (Variables|InternalData|GetPlayersThatHaveSpawnedAI) (Variables|InternalData|GetFort%20Player%20Pawn%20Athena))
  (if (Variables|Default|GetbDisableWhileSpawnActive)
    (Utilities|Array|AddUnique (Variables|InternalData|GetSpawnedPawns) _newpawn)
    (Game|BindEventtoOnDestroyed _newpawn (EventDispatchers|CreateEvent :!event OnPawnDestroyed))
    (Development|PrintString (Utilities|String|Append "@@@ On Pawn Spawned [BP_DebugNPCSpawner]: CALLED, Spawn count is: " (Utilities|String|ToString(Integer) (Utilities|Array|Length (Variables|InternalData|GetSpawnedPawns))) "!") false (Variables|Debug|GetbDebugSpawner)))
  (CallFunction|UpdateSpawnDisplay)
  ;[1 1 1] Check if this Request ID was requested by us
  (if (Utilities|Array|ContainsItem (Variables|InternalData|GetSpawnRequestIds) (Variables|GetRequestId))
    (if (Variables|Default|GetAutoHireAI)
      (Utilities|FlowControl|Sequence
        (:then_0
          ;[1 1 1] If IsHired bool is set to true, spawned NPC will be hired.
          (Networking|SwitchHasAuthority
            (:Authority
              (bind _asfort_pawn (Utilities|Casting|CastToFortPawn (Variables|GetNewPawn)))
              (Utilities|IsValid (Variables|InternalData|GetFort%20Player%20Pawn%20Athena)
                (:Is%20Valid
                  (Convert|TriggerConversion (Variables|InternalData|GetFort%20Player%20Pawn%20Athena) _asfort_pawn))))))
        (:then_1
          _)))
    (Default|CallNewPawnSpawned (Variables|GetNewPawn))
    (bind _returnvalue (Utilities|Array|RemoveItem (Variables|InternalData|GetSpawnRequestIds) (Variables|GetRequestId)))
    (if _returnvalue
      (CallFunction|ConfigureSpawnedPawn (Variables|GetNewPawn)))))

(fn OnRep_bDisabledIsVisible
  (Networking|SwitchHasAuthority
    (:Remote
      (Development|PrintString (Utilities|String|Append "@@@@@ On Rep bDisabledIsVisible [BP_DebugNPCSpawner]: CALLED, bDisabledIsVisible = " (Utilities|String|ToString(Boolean) (Variables|InternalData|GetbDisabledIsVisible)) "!") false (Variables|Debug|GetbDebugSpawner))
      (Rendering|SetVisibility (Variables|Default|GetTextRender_X) (Variables|InternalData|GetbDisabledIsVisible) false))))

(fn SpawnAndConfigureAI
  (bind _spawnedcount (Variables|InternalData|GetSpawnedCount))
  (bind _offsetperspawn (Variables|Default|GetOffsetPerSpawn))
  (bind _returnvalue (* _offsetperspawn _spawnedcount))
  (bind _actorlocation (Variables|Default|GetActorLocation))
  (bind _returnvalue_2 (Utilities|IsValid _actorlocation :!class /Script/Engine.KismetSystemLibrary))
  (bind _rootcomponent (Variables|/Script/Engine.Actor|GetRootComponent _actorlocation))
  (bind _defaultsceneroot (Variables|Default|GetDefaultSceneRoot))
  (bind _returnvalue_3 (select _returnvalue_2 _rootcomponent _defaultsceneroot))
  (bind _returnvalue_4 (Transformation|GetWorldRotation _returnvalue_3))
  (bind _returnvalue_5 (Transformation|GetWorldLocation _returnvalue_3))
  (bind _returnvalue_6 (+ _returnvalue_5 _returnvalue))
  (bind _returnvalue_7 (Math|Transform|MakeTransform _returnvalue_6 _returnvalue_4))
  ;[1 1 1] Get Athena AI Spawner
  (bind _returnvalue_8 (AI|GetAISpawner))
  (Utilities|IsValid _returnvalue_8
    (:Is%20Valid
      (bind _output_get (Variables|InternalData|SetAISpawner _returnvalue_8))
      (EventDispatchers|BindEventtoOnPawnSpawned _output_get (EventDispatchers|CreateEvent :!event OnPawnSpawned))
      (if (Variables|Default|GetUseStringBasedSpawning)
        ;[1 1 1] Spawn Using Console command instead of AI spawner
        (bind _returnvalue_9 (Class|FortAthenaLivingWorldManager|DebugSpawnEventData (Class|FortAthenaLivingWorldManager|GetLivingWorldManager) _returnvalue_7 (Variables|Default|GetStringBasedSpawning_ContentName)))
        (Math|Integer|IncrementInt (Variables|InternalData|GetSpawnedCount)
          (:%20%20
            (Utilities|Array|Add (Variables|InternalData|GetSpawnRequestIds) _returnvalue_9)))
        (else
          ;[1 1 1] Spawn
          (bind _returnvalue_10 (Class|FortAthenaAISpawnerData|CreateComponentListfromClass (Variables|Default|GetAIBotSpawnerData%20Class) self))
          (bind _modifiedcomponentlist (CallFunction|ModifySpawnerComponent _returnvalue_10))
          (bind _returnvalue_11 (Class|AthenaAISpawner|RequestSpawn (Variables|InternalData|GetAISpawner) _modifiedcomponentlist _returnvalue_7 false -1))
          (Math|Integer|IncrementInt (Variables|InternalData|GetSpawnedCount)
            (:%20%20
              (Utilities|Array|Add (Variables|InternalData|GetSpawnRequestIds) _returnvalue_11))))))
    (:Is%20Not%20Valid
      (Development|PrintString "AI Spawn Error: We can't find the AISpawner to spawn NPC , please make sure that \"Bots Enabled\" is checked in \"Editor Preferences > Battle Royale Cheats > Bots\"." true true "(R=1.000000,G=0.000000,B=0.000000,A=1.000000)" 1e+01))))

(fn ConstructionScript
  (CallFunction|GenerateDisplayTextfromCharacterDefinition))

(fn ValidatePlayerAlreadySpawnedAIOrNot
  (return (not (Utilities|Array|ContainsItem (Variables|InternalData|GetPlayersThatHaveSpawnedAI) (Variables|InternalData|GetFort%20Player%20Pawn%20Athena)))))
